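# Sound cue for Octane vehicle engine audio, playing LowSpec MetaSound preset sources as high-priority in-game SFX.
# A quality-level node sends the first level to one branch and the other three to a second, each with its own wave player.
# Each branch reads IsLocal, routing true through an Attenuation node and false or unset straight to the player.
# The cue itself uses the shared DM engine attenuation asset and the Fortnite SFX submix.
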SoundCue /VehicleAudio/EngineAudio/Octane/Octane_Preset_SC
root=QualityLevel_0
nodes [
 {id=Attenuation_0, children [WavePlayer_1]}
 {id=Attenuation_1, children [WavePlayer_2]}
 {id=Branch_1, children [Attenuation_0, WavePlayer_1, WavePlayer_1], properties {boolParameterName=IsLocal}}
 {id=Branch_2, children [Attenuation_1, WavePlayer_2, WavePlayer_2], properties {boolParameterName=IsLocal}}
 {id=QualityLevel_0, children [Branch_1,Branch_2,Branch_2,Branch_2]}
 {id=WavePlayer_1, properties {soundWaveAssetPtr=&/MSS_EngineAudio_Octane_LowSpec_Preset}}
 {id=WavePlayer_2, properties {soundWaveAssetPtr=&MSS_EngineAudio_Octane_LowSpec_Preset}}
]
properties {
 attenuationSettings=/VehicleAudio/Attenuation/DM_EngineAudio_Attn
 maxDistance=2097152
 priority=90
 soundClassObject=&Fort_SFX_InGame
 soundSubmixObject=&Fortnite_SFX
 volumeMultiplier=1
}
